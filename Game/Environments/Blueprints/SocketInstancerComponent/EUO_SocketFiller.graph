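# EUO_SocketFiller is an editor utility helper for the ISM_SocketFiller that owns it, deferring that object's construct logic and reading source properties.
# RunConstruct coalesces calls into one next-tick timer, and HandleConstructTimer clears it and calls DoActualConstructLogic on the outer object.
# Getter functions use GetEditorProperty to read floats, mesh structs and the RandomizedMeshes list by name.
# GetMeshDataFromSource is a stub, and GetMeshListFromSource contains apparently stray selection and subobject-creation nodes.
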EditorUtilityBlueprint /Game/Environments/Blueprints/SocketInstancerComponent/EUO_SocketFiller
graphs [
 EventGraph {nodes [
   Event RunConstruct() {is=title K2Node_Event}
   > n1 = Branch(Condition=Is Valid Timer Handle!(Handle:Timer Handle Structure=Get NextTickTimerHandle() -> (NextTickTimerHandle:Timer Handle Structure)) -> (ReturnValue:bool))
   n2 = Set Timer for Next Tick by Event!(execute=@n1.else, Delegate:Delegate=Create Event()) -> (then, ReturnValue:Timer Handle Structure) {is=title K2Node_CallFunction}
   > Set NextTickTimerHandle(NextTickTimerHandle:Timer Handle Structure=@n2) -> (Output_Get:Timer Handle Structure)
  ]}
 GetFloatPropertyFromSource {nodes [
   n1 = GetFloatPropertyFromSource!() -> (then, source:Object Reference, PropertyName:Name) {is=title K2Node_FunctionEntry}
   > n2 = GetEditorProperty!(execute:Exec, Object:Object Reference=@n1, PropertyName:Name=@n1.PropertyName) -> (then:Exec, PropertyValue:double, ReturnValue:bool)
   > Return Node!(execute, Success?:bool=@n2.ReturnValue, return:double=@n2)
  ]}
 GetMeshDataFromSource {nodes [
   GetMeshDataFromSource!() -> (then, Source:Object Reference, Property:Name=None) {is=title K2Node_FunctionEntry}
   > Return Node!(execute, Valid:bool=false, Data:Socket Filler Mesh Struct Structure)
  ]}
 GetMeshListFromSource {nodes [
   n1 = GetAllLevelActors(self=Get EditorActorSubsystem())
   > Filter by Class!(execute, TargetArray:Array of Object References=@n1, ObjectClass:Object Class Reference, FilterType:EEditorScriptingFilterType Enum) -> (then, ReturnValue:Array of Object References)
   n2 = GetSelectionSet(self=Get LevelEditorSubsystem())
   > n3 = GetSelectedObjects!(execute, self:Typed Element Selection Set Object Reference=@n2, InRequiredClass:Object Class Reference=&Actor) -> (then, ReturnValue:Array of Object References)
   > For Each Loop(Array:Array of Object References=@n3) -> (Array Element:Object Reference)
   > n4 = Gather Subobject Data For Instance(self=Get SubobjectDataSubsystem())
   > n5 = For Each Loop with Break!(Exec:Exec, Array:Array of Subobject Data Handle Structures=@n4, Break:Exec) -> (LoopBody:Exec, Array Element:Subobject Data Handle Structure, Array Index:int, Completed:Exec)
   n6 = GetData(DataHandle=@n5)
   > Branch(Condition=IsRootComponent!(Data:"Subobject Data Structure (by ref)"=@n6) -> (ReturnValue:bool))
   > AddNewSubobject!(
    execute
    self:Subobject Data Subsystem Object Reference=Get SubobjectDataSubsystem()
    Params:"Add New Subobject Params Structure (by ref)"=Make Add New Subobject Params!(
     ParentHandle:Subobject Data Handle Structure=GetHandle!(Data:"Subobject Data Structure (by ref)"=@n6) -> (OutHandle:Subobject Data Handle Structure)
     NewClass:Object Class Reference=/Script/Engine.SceneComponent
     BlueprintContext:Blueprint Object Reference
     bSkipMarkBlueprintModified:bool=False
     bConformTransformToParent:bool=True
    ) -> (AddNewSubobjectParams:Add New Subobject Params Structure) {is=title K2Node_MakeStruct}
   ) -> (then, FailReason:Text, ReturnValue:Subobject Data Handle Structure)
   n7 = GetMeshListFromSource!() -> (then, Source:Object Reference) {is=title K2Node_FunctionEntry}
   > n8 = GetEditorProperty(Object=@n7, PropertyName=RandomizedMeshes) -> (PropertyValue:Array of Socket Filler Mesh Struct Structures)
   > Return Node!(execute, Valid=@n8.ReturnValue, MeshList:Array of Socket Filler Mesh Struct Structures=@n8)
  ]}
 GetMeshPropertyFromSource {nodes [
   n1 = GetMeshPropertyFromSource!() -> (then, source:Object Reference, PropertyName:Name) {is=title K2Node_FunctionEntry}
   > n2 = GetEditorProperty(Object=@n1, PropertyName=@n1.PropertyName) -> (PropertyValue:Socket Filler Mesh Struct Structure)
   > Return Node!(execute, Success?:bool=@n2.ReturnValue, return:Socket Filler Mesh Struct Structure=@n2)
  ]}
 HandleConstructTimer {nodes [
   HandleConstructTimer() {is=title K2Node_FunctionEntry}
   > Clear and Invalidate Timer by Handle!(execute, WorldContextObject:Object Reference=Self-Reference(), Handle=Get NextTickTimerHandle()) -> (then)
   > n1 = Cast To ISM_SocketFiller!(execute, Object=GetOuterObject!(Object:Object Reference=Self-Reference()) -> (ReturnValue:Object Reference)) -> (then, CastFailed, AsISM Socket Filler:ISM Socket Filler Object Reference) {is=title K2Node_DynamicCast}
   > DoActualConstructLogic(self:ISM Socket Filler Object Reference=@n1)
  ]}
]
variables [NextTickTimerHandle {type=TimerHandle}]
