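; Instanced static mesh component that fills every socket of its parent (or owner root) mesh component with instances of one foliage mesh.
; Construct logic filters a mesh list from an editor utility or its own array, then picks a random or user-set index.
; Each socket transform gets stream-driven rotation, offset and scale, with the seed optionally derived from world position.
; Reconstruct requests route through a soft-loaded editor utility via the ISocketFillerEditorBridge interface.
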(event Custom|LoadFoliageMesh
  (bind _returnvalue (Utilities|LoadAssetBlocking (Utilities|Struct|BreakSocketFillerMeshStruct (Utilities|Array|Get(acopy) (Variables|Private|GetPrivateRandomMeshes) (Variables|SocketFiller|Mesh|GetCurrentMeshIndex)))))
  (bind _asstatic_mesh (Utilities|Casting|CastToStaticMesh _returnvalue))
  (Components|StaticMesh|SetStaticMesh _asstatic_mesh))

(fn DoActualConstructLogic
  (Components|InstancedStaticMesh|ClearInstances)
  (if (Variables|SocketFiller|Random|GetRandomizeSeedByPosition)
    (Random|GenerateRandomStreamfromPosition))
  (Utilities|FlowControl|Sequence
    (:then_0
      (CallFunction|ValidateMeshList)
      (if (and (Variables|SocketFiller|GetIsEnabled) (Utilities|Array|IsValidIndex (Variables|Private|GetPrivateRandomMeshes)))
        (bind _returnvalue (CallFunction|GetTargetMeshComponent))
        (bind _output_get (Variables|Private|SetTargetMeshComponent _returnvalue))
        (Utilities|IsValid _output_get
          (:Is%20Valid
            (if (Variables|SocketFiller|Mesh|GetRandomizeMeshIndex)
              (Random|GetRandomMeshIndex)
              (else
                (Default|ValidateUserMeshIndex) ; If user typed a value, make sure it's valid
              ))
            (CallFunction|LoadFoliageMesh)
            (Components|InstancedStaticMesh|ClearInstances)
            (bind _sockettransforms (Default|GetSocketTransforms))
            (Components|InstancedStaticMesh|AddInstances _sockettransforms false false false)))))
    (:then_1
      (Variables|Private|SetCurrentEditorUtil 0))))

(fn GenerateRandomStreamFromPosition
  (bind _returnvalue (Transformation|GetWorldLocation))
  (bind _randomizationtransitiondistance (Variables|SocketFiller|Random|GetRandomizationTransitionDistance))
  (bind _z (.z _returnvalue))
  (bind _returnvalue_2 (Math|Float|Absolute(Float) _z))
  (bind _y (.y _returnvalue))
  (bind _returnvalue_3 (Math|Float|Absolute(Float) _y))
  (bind _x (.x _returnvalue))
  (bind _returnvalue_4 (Math|Float|Absolute(Float) _x))
  (bind _returnvalue_5 (Math|Float|float+float _returnvalue_4 _returnvalue_3 _returnvalue_2))
  (bind _returnvalue_6 (/ _returnvalue_5 3.0))
  (bind _returnvalue_7 (/ _returnvalue_6 _randomizationtransitiondistance))
  (bind _returnvalue_8 (Math|Float|Round _returnvalue_7))
  (bind _randomseed (Variables|SocketFiller|GetRandomSeed))
  (bind _returnvalue_9 (+ _returnvalue_8 _randomseed))
  (Variables|SocketFiller|Random|SetCurrentRandomSeed _returnvalue_9)
  (Variables|Private|SetCurrentRandomStream (Math|Random|MakeRandomStream _returnvalue_9)))

(fn GetRandomMeshIndex
  (bind _output_get (LocalVariables|Setl_meshCount (Utilities|Array|Length (Variables|Private|GetPrivateRandomMeshes))))
  (if (> _output_get 0)
    (LocalVariables|Setl_Location (Transformation|GetWorldLocation))
    (Utilities|FlowControl|Sequence
      (:then_0
        (Variables|SocketFiller|Mesh|SetCurrentMeshIndex (Math|Random|RandomIntegerinRangefromStream (Variables|Private|GetCurrentRandomStream) 0 (- (LocalVariables|Getl_meshCount) 1))))
      (:then_1
        (return (Variables|SocketFiller|Mesh|GetCurrentMeshIndex))))))

(fn GetEditorUtil
  (if (Editor|InEditororPIE)
    (Utilities|FlowControl|Sequence
      (:then_0
        (Utilities|IsValid (Variables|Private|GetCurrentEditorUtil)
          (:Is%20Not%20Valid
            (if (Utilities|IsValidSoftClassReference (Variables|Private|GetEditorUtilClass))
              (bind _returnvalue (Utilities|LoadClassAssetBlocking (Variables|Private|GetEditorUtilClass)))
              (bind _returnvalue_2 (Game|ConstructObjectfromClass _returnvalue))
              (Variables|Private|SetCurrentEditorUtil _returnvalue_2)))))
      (:then_1
        (return (Variables|Private|GetCurrentEditorUtil))))))

(fn GetSocketTransforms
  (for _array_element (Transformation|GetAllSocketNames (Variables|Private|GetTargetMeshComponent))
    (bind _returnvalue (Default|ValidateTransformforInstance (Transformation|GetSocketTransform (Variables|Private|GetTargetMeshComponent) _array_element "RTS_Component")))
    (Utilities|Array|Add (LocalVariables|Getl_socketTransforms) _returnvalue))
  (return (LocalVariables|Getl_socketTransforms)))

(fn getTargetMeshComponent
  (bind _asmesh_component (Utilities|Casting|CastToMeshComponent (Transformation|GetAttachParent))
    (:then
      (return _asmesh_component))
    (:CastFailed
      (bind _asmesh_component_2 (Utilities|Casting|CastToMeshComponent (Variables|/Script/Engine.Actor|GetRootComponent (Components|GetOwner))))
      (return _asmesh_component_2))))

(fn ReconstructSocketFiller
  (Construct|RunConstruct))

(fn Refresh
  (Construct|RunConstruct))

(fn RunConstruct
  (if (Editor|InEditor)
    (Class|ISocketFillerEditorBridge|RunConstruct(Message) (Editor|GetEditorUtil))))

(fn ValidateMeshList
  (Utilities|Array|Clear (Variables|Private|GetPrivateRandomMeshes))
  (bind (_valid _meshlist) (Default|GetMeshListfromSource(Message) (Editor|GetEditorUtil) self))
  (if _valid
    (LocalVariables|Setl_meshList _meshlist)
    (else
      (LocalVariables|Setl_meshList (Variables|SocketFiller|Mesh|GetRandomizedMeshes))))
  (for _array_element (LocalVariables|Getl_meshList)
    (if (and (Utilities|IsValidSoftObjectReference (Utilities|Struct|BreakSocketFillerMeshStruct _array_element)) (!= (.ScaleMultiplier_6_9A61CEDB4EF183CA229418ADD4D418C4 (Utilities|Struct|BreakSocketFillerMeshStruct _array_element)) 0))
      (Utilities|Array|Add (Variables|Private|GetPrivateRandomMeshes) _array_element))))

(fn ValidateTransformForInstance
  (params (InTransform :type Transform%20%28by%20ref%29))
  (bind (_staticmesh_2_1ebb6cab499d847440647faba8135982 _scalemultiplier_6_9a61cedb4ef183ca229418add4d418c4) (Utilities|Struct|BreakSocketFillerMeshStruct (Utilities|Array|Get(acopy) (Variables|Private|GetPrivateRandomMeshes) (Variables|SocketFiller|Mesh|GetCurrentMeshIndex))))
  (bind (_x _y) (Math|Vector2D|BreakVector2D (Variables|SocketFiller|Scale|GetRandomScaleRange)))
  (LocalVariables|Setl_transform InTransform)
  (LocalVariables|Setl_scaleMultiplier _scalemultiplier_6_9a61cedb4ef183ca229418add4d418c4)
  (Utilities|FlowControl|Sequence
    (:then_0
      (LocalVariables|Setl_rotRand (Utilities|Select :Index (Variables|SocketFiller|Rotation|GetRandomizeRotationAxis) :NewEnumerator0 "0, 0, 0" :NewEnumerator1 (Math|Rotator|MakeRotator (Math|Random|RandomFloatinRangefromStream (Variables|Private|GetCurrentRandomStream) 0.0 3.6e+02)) :NewEnumerator2 (Math|Rotator|MakeRotator 0.0 (Math|Random|RandomFloatinRangefromStream (Variables|Private|GetCurrentRandomStream) 0.0 3.6e+02)) :NewEnumerator3 (Math|Rotator|MakeRotator 0.0 0.0 (Math|Random|RandomFloatinRangefromStream (Variables|Private|GetCurrentRandomStream) 0.0 3.6e+02))))
      (LocalVariables|Setl_rotFinal (select (Variables|SocketFiller|Rotation|GetUseSocketRotation) (Math|Rotator|InvertRotator (Math|Rotator|CombineRotators (Math|Rotator|InvertRotator (.rotation (LocalVariables|Getl_transform))) (LocalVariables|Getl_rotRand))) (Math|Rotator|CombineRotators (Transformation|GetWorldRotation (Variables|Private|GetTargetMeshComponent)) (Math|Rotator|CombineRotators (Math|Rotator|InvertRotator (Transformation|GetWorldRotation (Variables|Private|GetTargetMeshComponent))) (LocalVariables|Getl_rotRand))))))
    (:then_1
      (LocalVariables|Setl_locOffset (Math|Vector|MakeVector (Math|Random|RandomFloatinRangefromStream (Variables|Private|GetCurrentRandomStream) (* (.x (Math|Vector|VectorGetAbs (Variables|SocketFiller|Location|GetRandomLocationOffset))) -1.0) (.x (Math|Vector|VectorGetAbs (Variables|SocketFiller|Location|GetRandomLocationOffset)))) (Math|Random|RandomFloatinRangefromStream (Variables|Private|GetCurrentRandomStream) (* (.y (Math|Vector|VectorGetAbs (Variables|SocketFiller|Location|GetRandomLocationOffset))) -1.0) (.y (Math|Vector|VectorGetAbs (Variables|SocketFiller|Location|GetRandomLocationOffset)))) (Math|Random|RandomFloatinRangefromStream (Variables|Private|GetCurrentRandomStream) (* (.z (Math|Vector|VectorGetAbs (Variables|SocketFiller|Location|GetRandomLocationOffset))) -1.0) (.z (Math|Vector|VectorGetAbs (Variables|SocketFiller|Location|GetRandomLocationOffset))))))
      (LocalVariables|Setl_locFinal (+ (.location (LocalVariables|Getl_transform)) (Math|Vector|RotateVector (LocalVariables|Getl_locOffset) (.rotation (LocalVariables|Getl_transform))))))
    (:then_2
      (LocalVariables|Setl_scaleFinal (select (Variables|SocketFiller|Scale|GetUseSocketScale) (* (select (Variables|SocketFiller|Scale|GetRandomizeScale) (* (LocalVariables|Getl_scaleMultiplier) (Math|Random|RandomFloatinRangefromStream (Variables|Private|GetCurrentRandomStream) _x _y)) (LocalVariables|Getl_scaleMultiplier)) (.scale (LocalVariables|Getl_transform))) (Math|Conversions|ToVector(Float) (select (Variables|SocketFiller|Scale|GetRandomizeScale) (* (LocalVariables|Getl_scaleMultiplier) (Math|Random|RandomFloatinRangefromStream (Variables|Private|GetCurrentRandomStream) _x _y)) (LocalVariables|Getl_scaleMultiplier))))))
    (:then_3
      (return (Math|Transform|MakeTransform (LocalVariables|Getl_locFinal) (LocalVariables|Getl_rotFinal) (LocalVariables|Getl_scaleFinal))))))

(fn ValidateUserMeshIndex
  (bind _l_listlength (LocalVariables|Getl_listLength))
  (bind _returnvalue (- _l_listlength 1))
  (bind _currentmeshindex (Variables|SocketFiller|Mesh|GetCurrentMeshIndex))
  (bind _returnvalue_2 (Math|Integer|Min(Integer) _currentmeshindex _returnvalue))
  (bind _returnvalue_3 (Math|Integer|Max(Integer) _returnvalue_2))
  (bind _output_get (LocalVariables|Setl_listLength (Utilities|Array|Length (Variables|Private|GetPrivateRandomMeshes))))
  (if (> _output_get 0)
    (if (!= _returnvalue_3 (Variables|SocketFiller|Mesh|GetCurrentMeshIndex))
      (Variables|SocketFiller|Mesh|SetCurrentMeshIndex _returnvalue_3))))
